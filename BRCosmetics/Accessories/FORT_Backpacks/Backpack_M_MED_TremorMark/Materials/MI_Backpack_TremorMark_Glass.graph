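# Translucent lit glass base Material (despite its MI_ prefix) for the TremorMark backpack cosmetic, with refraction and skeletal mesh support.
# A scaled noise normal feeds a Fresnel that blends between inner and outer transparency.
# Fake reflections come from an HDR sky cubemap tinted by the ambient cubemap and partly desaturated.
# A sharp highlight from the shared sun/moon direction is added to them, then passed through an FX emissive adjustment function.
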Material /BRCosmetics/Accessories/FORT_Backpacks/Backpack_M_MED_TremorMark/Materials/MI_Backpack_TremorMark_Glass
graph {
 properties {
  bEnableResponsiveAA=true
  bScreenSpaceReflections=true
  bUsedWithMeshDeformer=true
  bUsedWithSkeletalMesh=true
  blendMode=BLEND_Translucent
  physMaterial=/Game/FortressPhysicalMaterials/Glass
  pixelDepthOffsetMode=PDOM_Legacy
  refractionMethod=RM_IndexOfRefraction
  translucencyDirectionalLightingIntensity=5
  translucencyLightingMode=TLM_VolumetricPerVertexNonDirectional
  usedShadingModels=Default Lit
 }
 nodes [
  CollectionParameter() {collection=&FortniteMaterialParameters, parameterName=SunLightColor}
  n1 = TextureCoordinate() {vTiling=0.2}
  n2 = Normalize(VectorInput=Multiply(
    A=TextureSample(UVs=@n1) {samplerType=SAMPLERTYPE_Normal, texture=/Game/Global/Textures/Default/Noise/T_Noise_Normal_01}.RGB
    B=VectorParameter() {parameterName=Normal Intensity, defaultValue=<rgba 0.1 0.1 1 0>}.RGB
   ))
  LinearInterpolate(
   A=ScalarParameter() {parameterName=Fresnel Inner Transparency}
   B=ScalarParameter() {defaultValue=0.85, parameterName=Fresnel Outer Transparency}
   Alpha=Fresnel(ExponentIn=ScalarParameter() {defaultValue=4, parameterName=Fresnel Exponent}, Normal=Transform(None=@n2))
  )
  n3 = ReflectionVectorWS() {bNormalizeCustomWorldNormal=true}
  n4 = Multiply(
   A=Multiply(
    A=TextureSampleParameterCube!(UVs=@n3, Apply View MipBias) {parameterName=MaterialExpressionTextureSampleParameterCube_0, texture=/Game/Packages/Fortress_Sky/TexturesHDR/T_hdrsky_test03_low}.RGB
    B=VectorParameter() {parameterName=Cube Map Multiplier, defaultValue=<rgba 1.5 1.5 1.5 0>}.RGB
   ) {constB=1.5}
   B=ComponentMask(None=MaterialFunctionCall() {materialFunction=/Engine/Functions/GetPostProcessFunctions/GetAmbientCubemapTint}.AmbientCubemapTint) {b=true, g=true, r=true}
  )
  n5 = Constant() {r=0.5}
  n6 = CollectionParameter() {
   collection=&FortniteMaterialParameters
   parameterName=SunAndMoonModelDirectionalVector
  }
  n7 = ComponentMask(None=@n6) {b=true, g=true, r=true}
  n8 = ReflectionVectorWS()
  n9 = Power(Base=Clamp(None=DotProduct(A=@n7, B=@n8)), Exp=ScalarParameter() {defaultValue=70, parameterName=Spec Highlight Power}) {constExponent=20}
  n10 = ScalarParameter() {defaultValue=3, parameterName=Spec Intense}
  MaterialFunctionCall(Emissive=Add(A=Desaturation(None=@n4, Fraction=@n5), B=Multiply(A=@n9, B=@n10))) {materialFunction=&MF_FX_EmissiveAdjustments}
  ScalarParameter() {defaultValue=1, parameterName=Refraction, sliderMax=2}
  ScalarParameter() {defaultValue=0.1, parameterName=Metallic}
  ScalarParameter() {defaultValue=1, parameterName=Specular}
  VectorParameter() {parameterName=Base Color}
 ]
}
